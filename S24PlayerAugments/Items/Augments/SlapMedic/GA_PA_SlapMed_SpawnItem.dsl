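; Passive gameplay ability for the Slap Medic player augment that triggers when its owner searches an allowed container, apparently to spawn an item.
; On activation it binds to the avatar pawn's server searched-container event.
; On search, it checks the container's tags against allowed types, stores the container and calls ActivateAugment.
; GetLocationToSpawn returns a point above and in front of that container, with random horizontal scatter.
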(event Ability|EventFailedToActivatePassiveAbility
  (params (ActorInfo :type Gameplay%20Ability%20Actor%20Info%20Structure)))

(event Ability|EventActivateAbility
  (EventDispatchers|BindEventtoServer_OnSearchedContainer (CallFunction|GetAvatarasFortPlayerPawn) (delegate Custom|OnSearch)))

(event Custom|OnSearch
  (params (SearchingPawn :type Fort%20Pawn%20Object%20Reference) (Container :type Building%20Container%20Object%20Reference))
  (if (GameplayTags|HasAnyMatchingGameplayTags Container (Variables|Default|GetAllowedContainerTypes))
    (Variables|Default|SetContainer Container)
    (CallFunction|ActivateAugment)))

(fn GetLocationToSpawn
  (return (Utilities|Operators|Add (Transformation|GetActorLocation (Variables|Default|GetContainer)) (Math|Vector|MakeVector (Math|Random|RandomFloatinRange -25.0 25.0) (Math|Random|RandomFloatinRange -25.0 25.0) 5e+01) (* (Transformation|GetActorForwardVector (Variables|Default|GetContainer)) 5e+01))))
